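# First-person music cue for the ShallWe emote, built from four looping layers that stack as a group count rises.
# After a short fixed delay, a parameter crossfade driven by GroupNum holds layer A as the base and fades in B, C and D over successive steps.
# The cue uses the first-person emote music sound class and concurrency set, and sends to the music audio-analysis submix.
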SoundCue /BRCosmetics/Sounds/Emotes/ShallWe/Emote_ShallWe_1P_Cue
root=Delay_1
nodes [
 {id=Delay_1, children [ParamCrossFade_0], properties {delayMax=0.25, delayMin=0.25}}
 {
  id=ParamCrossFade_0
  children [WavePlayer_0,WavePlayer_4,WavePlayer_5,WavePlayer_6]
  properties {
   paramName=GroupNum
   crossFadeInput [
    {fadeInDistanceEnd=0, fadeInDistanceStart=0, fadeOutDistanceEnd=100, fadeOutDistanceStart=100, volume=1}
    {fadeInDistanceEnd=2, fadeInDistanceStart=1, fadeOutDistanceEnd=100, fadeOutDistanceStart=100, volume=1}
    {fadeInDistanceEnd=3, fadeInDistanceStart=2, fadeOutDistanceEnd=100, fadeOutDistanceStart=100, volume=1}
    {fadeInDistanceEnd=4, fadeInDistanceStart=3, fadeOutDistanceEnd=100, fadeOutDistanceStart=100, volume=1}
   ]
  }
 }
 {id=WavePlayer_0, properties {bLooping=true, soundWaveAssetPtr=&/Emote_ShallWe_Layer_A}}
 {id=WavePlayer_4, properties {bLooping=true, soundWaveAssetPtr=&/Emote_ShallWe_Layer_B}}
 {id=WavePlayer_5, properties {bLooping=true, soundWaveAssetPtr=&/Emote_ShallWe_Layer_C}}
 {id=WavePlayer_6, properties {bLooping=true, soundWaveAssetPtr=&/Emote_ShallWe_Layer_D}}
]
properties {
 maxDistance=2097152
 soundClassObject=/Game/Sounds/FortSoundClassesAndModes/Fort_Music_Emote_1P_PSM
 volumeMultiplier=1
 concurrencyOverrides {maxCount=3, resolutionRule=StopOldest}
 concurrencySet [/Game/Sounds/Concurrency_Rules/Athena_EmoteMusic_1P_Conc]
 soundSubmixSends [{minSendLevel=0, soundSubmix=&Fortnite_AudioAnalysis_Music}]
}
