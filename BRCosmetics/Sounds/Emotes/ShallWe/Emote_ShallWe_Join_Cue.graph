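# Sound cue for the Join audio of the Shall We emote, classed as emote music.
# A single wave player feeds a modulator left at default settings, with the cue's overall volume reduced.
# Playback uses shared emote attenuation and a dedicated concurrency set, alongside override values that favor stopping the oldest instance.
# The cue also declares a send to the music audio-analysis submix.
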SoundCue /BRCosmetics/Sounds/Emotes/ShallWe/Emote_ShallWe_Join_Cue
root=Modulator_0
nodes [
 {id=Modulator_0, children [WavePlayer_1]}
 {id=WavePlayer_1, properties {soundWaveAssetPtr=&/Emote_ShallWe_Join_01}}
]
properties {
 attenuationSettings=&Fort_Emotes_Athena_Att
 maxDistance=1600
 soundClassObject=&Fort_Music_Emote
 volumeMultiplier=0.5
 concurrencyOverrides {maxCount=3, resolutionRule=StopOldest}
 concurrencySet [&/EmoteShall_Concurrency]
 soundSubmixSends [{minSendLevel=0, soundSubmix=&Fortnite_AudioAnalysis_Music}]
}
